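; Touch-collected quest item that credits a player's quest objective when they overlap its collect volume.
; The construction script copies quest, objective name, blocked playlist tags and a test force-active flag into the QuestInteractable component.
; The server binds the overlap and sends a stat event to the overlapping actor's quest manager if that objective is in progress.
; Clients toggle the volume's collision as the quest is enabled or disabled.
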(event SetupforServer(QuestInteractable)
  ;[1 1 1] SERVER: Setup
  (Collision|BindEventtoOnComponentBeginOverlap (Variables|Default|GetCollectOverlap) (EventDispatchers|CreateEvent :!event Custom|Collected)))

(event ClientQuestEnabled(QuestInteractable)
  ;[1 1 1] CLIENT: QUEST ENABLED
  (Collision|SetCollisionEnabled (Variables|Default|GetCollectOverlap) "QueryAndPhysics"))

(event ClientQuestDisabled(QuestInteractable)
  ;[1 1 1] CLIENT: QUEST DISABLED
  (Collision|SetCollisionEnabled (Variables|Default|GetCollectOverlap) "NoCollision"))

(event Custom|Collected
  (params (OverlappedComponent :type Primitive%20Component%20Object%20Reference) (OtherActor :type Actor%20Object%20Reference) (OtherComp :type Primitive%20Component%20Object%20Reference) (OtherBodyIndex :type Integer) (bFromSweep :type Boolean) (SweepResult :type Hit%20Result%20Structure%20%28by%20ref%29))
  ;[1 1 1] SERVER ONLY: COLLECTION
  (bind _returnvalue (Class|QuestInteractableComponent|GetQuestManagerforActor (Variables|Default|GetQuestInteractable) OtherActor))
  (Utilities|IsValid _returnvalue
    (:Is%20Valid
      (if (Quest|IsObjectivewithNameinProgress _returnvalue (Variables|Default|GetCollectItemQuest) (Variables|Default|GetQuestBackendName) :!class /Script/FortniteGame.FortQuestManager)
        (Quest|SendComplexCustomStatEvent _returnvalue 0 (Variables|Default|GetObjStatTag) "(GameplayTags=)" 1)))))

(fn ConstructionScript
  (bind _questinteractable (Variables|Default|GetQuestInteractable))
  (Variables|/Script/FortniteGame.QuestInteractableComponent|SetQuestItemDefinition (Variables|Default|GetCollectItemQuest) _questinteractable)
  (Variables|/Script/FortniteGame.QuestInteractableComponent|SetObjectiveBackendName (Variables|Default|GetQuestBackendName) _questinteractable)
  (Variables|/Script/FortniteGame.QuestInteractableComponent|SetBlockedPlaylistTags (Variables|Default|GetBlockedPlaylistTags) _questinteractable)
  (Variables|/Script/FortniteGame.QuestInteractableComponent|SetDeveloper_ForceActive (Variables|Default|GetTest_Item) _questinteractable))
