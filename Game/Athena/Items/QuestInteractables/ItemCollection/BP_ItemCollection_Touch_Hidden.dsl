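; Touch-collected quest item, using hidden XP coin visuals, that becomes collectible only after its hiding prop is destroyed.
; BeginPlay disables the collect overlap and binds to the prop's death, which starts a movement-duration timer to re-enable overlap queries.
; Each client-created visual is cached and bound to the prop with streak, glow and offset settings.
; A multicast on prop death notifies every cached visual.
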(event EventBeginPlay
  (|CallParentFunction)
  (Collision|SetCollisionEnabled (Variables|Default|GetCollectOverlap) "NoCollision")
  (EventDispatchers|BindEventtoOnDied (Variables|Default|GetHidingProp) (delegate Custom|OnHidingPropDestroyed)))

(event ClientQuestVisualCreated(QuestInteractable)
  (params (QuestVisual :type Quest%20Visual%20Object%20Reference))
  ;[1 1 1] Update the location of the quest visuals for each client.
  (bind _asquest_visual_xpcoin_hidden (Utilities|Casting|CastToQuestVisual_XPCoin_Hidden QuestVisual))
  (Utilities|Array|AddUnique (Variables|Default|GetQuestVisuals) _asquest_visual_xpcoin_hidden)
  (Class|QuestVisualXPCoinHidden|BindHideProp _asquest_visual_xpcoin_hidden (Variables|Default|GetHidingProp) (Variables|Default|GetVFXStreaks%20Height) (Variables|Default|GetVFXStreaks%20Radius) (Variables|Default|GetAdditionalStreaks) (Variables|Default|GetStreaksScaleMultiplier) (Variables|Default|GetSpawnOffset) (Variables|Default|GetVFXStreaks%20Offset) (Variables|Default|GetVFXGlowOffset) (Variables|Default|GetVFXGlow%20Scale)))

(event Custom|OnHidingPropDestroyed
  (params (DamagedActor :type Actor%20Object%20Reference) (Damage :type Float%20%28single-precision%29) (InstigatedBy :type Controller%20Object%20Reference) (DamageCauser :type Actor%20Object%20Reference) (HitLocation :type Vector) (FHitComponent :type Primitive%20Component%20Object%20Reference) (BoneName :type Name) (Momentum :type Vector))
  (bind _returnvalue (Utilities|Time|SetTimerbyEvent (delegate Custom|ReadyToCollect) (Variables|Default|GetMovementTimelineDuration)))
  (Variables|Default|SetCollectionWindowTimer _returnvalue)
  (CallFunction|MulticastPropDestroyed))

(event Custom|ReadyToCollect
  (Collision|SetCollisionEnabled (Variables|Default|GetCollectOverlap) "QueryOnly"))

(event Custom|MulticastPropDestroyed
  (for _array_element (Variables|Default|GetQuestVisuals)
    (Class|QuestVisualXPCoinHidden|HidingPropDestroyed _array_element)))

(fn setNiagaraSystemUserVariable(Color)
  (params (Name :type Name) (Linear%20Color :type Linear%20Color%20Structure))
  (ManagedParticles|Parameters|%28defunct%29SetColorParameter%28ManagedParticles%29 0 (ManagedParticles|GetNiagaraParticles) Name Linear%20Color))

(fn ConstructionScript)
