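; Quest XP coin visual that spins near players and plays a collection effect when its objective completes.
; Player overlap starts a spin-and-pulse timeline that a timer replays after a delay until overlap ends.
; Completion stops spinning, hides the light, and runs a timeline that moves, shrinks and glows the coin via dynamic materials, spawns owner-only particles, then sets a lifespan.
; The construction script configures the optional light using platform and effects-quality checks.
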(event EventBeginPlay
  ;[1 1 1] BEGIN PLAY
  (Components|Timeline|SetPlayRate (Variables|QuestVisual_XPCoin|GetPickedUp_TL) 2.5))

(event OnComponentBeginOverlap(SpinOverlap)
  (params (OverlappedComponent :type Primitive%20Component%20Object%20Reference) (OtherActor :type Actor%20Object%20Reference) (OtherComp :type Primitive%20Component%20Object%20Reference) (OtherBodyIndex :type Integer) (bFromSweep :type Boolean) (SweepResult :type Hit%20Result%20Structure%20%28by%20ref%29))
  ;[1 1 1] CLIENT ONLY SPINNING
  (Utilities|Casting|CastToFortPlayerPawnAthena OtherActor)
  (Variables|Default|SetCurrentRelativeScale (Variables|/Script/Engine.SceneComponent|GetRelativeScale3D (Variables|Default|GetSM_Coin)))
  (bind (_direction _risetime _pulsetime) (Timeline|SpinTL Play)
    (:Update
      (Transformation|SetRelativeRotation (Variables|Default|GetSM_Coin) (Math|Rotator|MakeRotator 0.0 0.0 (* _risetime (Variables|Default|GetRelativeRotationAmount))) false false)
      (Transformation|SetRelativeScale3D (Variables|Default|GetSM_Coin) (Math|Vector|Lerp(Vector) (Variables|Default|GetCurrentRelativeScale) (* (Variables|Default|GetCurrentRelativeScale) (Variables|Default|GetPulseSizeMultiplier)) _pulsetime)))
    (:Finished
      (bind _returnvalue (Utilities|Time|SetTimerbyEvent (Utilities|Select :Index _direction :Forward (delegate Custom|LoopForwards) :Backward (delegate Custom|LoopBackwards)) (Variables|Default|GetDelayBetweenSpin)))
      (Variables|Default|SetLoopTimerHandle _returnvalue))
    (:Start%20Spin
      (switch Utilities|FlowControl|Switch|SwitchonETimelineDirection _direction
        (:Forward
          (Audio|PlaySoundatLocation (Variables|Sound|GetCoinSpinSound) (Transformation|GetWorldLocation (Variables|Default|GetSM_Coin))))))))

(event Custom|LoopForwards
  ;[1 1 1] CLIENT ONLY SPINNING
  (Timeline|SpinTL PlayFromStart))

(event Custom|LoopBackwards
  ;[1 1 1] CLIENT ONLY SPINNING
  (Timeline|SpinTL ReverseFromEnd))

(event OnComponentEndOverlap(SpinOverlap)
  (params (OverlappedComponent :type Primitive%20Component%20Object%20Reference) (OtherActor :type Actor%20Object%20Reference) (OtherComp :type Primitive%20Component%20Object%20Reference) (OtherBodyIndex :type Integer))
  ;[1 1 1] CLIENT ONLY SPINNING
  (Components|Timeline|Stop (Variables|QuestVisual_XPCoin|GetSpinTL))
  (Utilities|Time|ClearandInvalidateTimerbyHandle (Variables|Default|GetLoopTimerHandle)))

(event Custom|EndSpinning
  ;[1 1 1] CLIENT ONLY SPINNING
  (Components|Timeline|Stop (Variables|QuestVisual_XPCoin|GetSpinTL))
  (Utilities|Time|ClearandInvalidateTimerbyHandle (Variables|Default|GetLoopTimerHandle)))

(event EventObjectiveCompleted
  (params (PlayerController :type Fort%20Player%20Controller%20Object%20Reference))
  (bind (_relativerotation_roll _relativerotation_pitch _relativerotation_yaw) (Math|Rotator|BreakRotator (Variables|/Script/Engine.SceneComponent|GetRelativeRotation_Roll (Variables|Default|GetSM_Coin))))
  ;[1 1 1] CLIENT ONLY: Collection FX
  (CallFunction|EndSpinning)
  (bind _returnvalue (Rendering|Material|CreateDynamicMaterialInstance (Variables|Default|GetSM_Coin) 0 0 "None" :!class /Script/Engine.PrimitiveComponent))
  (Variables|Default|SetMID0 _returnvalue)
  (bind _returnvalue_2 (Rendering|Material|CreateDynamicMaterialInstance (Variables|Default|GetSM_Coin) 1 0 "None" :!class /Script/Engine.PrimitiveComponent))
  (Variables|Default|SetMID1 _returnvalue_2)
  (Audio|PlaySoundatLocation (Variables|Sound|GetCoinCollectionSound) (Transformation|GetActorLocation))
  (Variables|Default|SetPreShrunkScale (Variables|/Script/Engine.SceneComponent|GetRelativeScale3D (Variables|Default|GetSM_Coin)))
  (Variables|Default|SetCoinRelativeTransform (Variables|/Script/Engine.SceneComponent|GetRelativeLocation (Variables|Default|GetSM_Coin)))
  (Variables|Default|SetCoinRelativeRotationZ _relativerotation_yaw)
  (bind _light (Variables|Default|GetLight)
    (:then
      (Development|SetHiddeninGame _light true false))
    (:else
      _))
  (bind (_direction _scale _zrotation _bobheight _location) (Timeline|PickedUp_TL PlayFromStart)
    (:Update
      (Transformation|SetRelativeLocation (Variables|Default|GetSM_Coin) (+ _location (Variables|Default|GetCoinRelativeTransform)) false false)
      (Transformation|SetRelativeScale3D (Variables|Default|GetSM_Coin) (Math|Vector|Lerp(Vector) (Variables|Default|GetPreShrunkScale) "0, 0, 0" _scale))
      (Transformation|AddRelativeRotation (Variables|Default|GetSM_Coin) (Math|Rotator|MakeRotator 0.0 0.0 (+ _zrotation (Variables|Default|GetCoinRelativeRotationZ))) false false)
      (Rendering|Material|SetVectorParameterValue (Variables|Default|GetMID0) "BobHeightMax" (Math|Color|Lerp(LinearColor) :B (Variables|Default|GetBobHeightMax) :Alpha _bobheight) :!class /Script/Engine.MaterialInstanceDynamic)
      (Rendering|Material|SetVectorParameterValue (Variables|Default|GetMID0) "BobHeightMin" (Math|Color|Lerp(LinearColor) :B (Variables|Default|GetBobHeightMin) :Alpha _bobheight) :!class /Script/Engine.MaterialInstanceDynamic)
      (Rendering|Material|SetScalarParameterValue (Variables|Default|GetMID0) "GlowScale" _bobheight :!class /Script/Engine.MaterialInstanceDynamic))
    (:Finished
      (Actor|SetLifeSpan 5.0)
      (Development|SetHiddeninGame (Variables|Default|GetSM_Coin) true true))
    (:SpawnParticles
      (bind _returnvalue_3 (Effects|SpawnEmitterAttached :EmitterTemplate (Variables|Default|GetCoinCollectionEmitter) :AttachToComponent (Variables|Default|GetDefaultSceneRoot) :AttachPointName "None" :Location (Transformation|GetSocketLocation (Variables|Default|GetSM_Coin) "FX") :Scale (/ (Variables|Default|GetPreShrunkScale) 0.75) :LocationType "KeepWorldPosition" :bAutoDestroy true :PoolingMethod "None" :bAutoActivate true))
      (Rendering|SetOnlyOwnerSee _returnvalue_3 true :!class /Script/Engine.PrimitiveComponent))))

(fn ConstructionScript
  (Networking|CanExecuteCosmeticEvents
    (:True
      (if (or (Variables|Default|GetUseLight) (not (Fortnite|IsMobilePlatform)))
        (Utilities|EffectsQualityCheck 0
          (:Cinematic
            _)
          (:Epic
            _)
          (:High
            _)
          (:Medium
            (Components|DestroyComponent (Variables|Default|GetLight)))
          (:Low
            _))
        (Rendering|Components|Light|SetLightColor (Variables|Default|GetLight) (Variables|Default|GetLightColor) true)
        (else
          _)))
    (:False
      _)))
