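; Cosmetic quest-interactable XP coin that hides at a building prop and is revealed when that prop is destroyed.
; BindHideProp stores VFX tuning and moves the hidden coin to the prop; HidingPropDestroyed flies it to its original location via timeline.
; Overlap runs a timer-looped spin timeline, spawning a parameterized Niagara flare while the prop is intact.
; Objective completion plays a shrinking pickup with material, sound and owner-only particles, then sets a short lifespan.
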(event Custom|BindHideProp
  (params (HidingProp :type Building%20Prop%20Object%20Reference) (VFXStreaks%20Height :type Float%20%28double-precision%29) (VFXStreaks%20Radius :type Float%20%28double-precision%29) (AdditionalStreaks :type Float%20%28double-precision%29) (StreaksScaleMultiplier :type Float%20%28double-precision%29) (SpawnOffset :type Vector) (VFXStreaks%20Offset :type Vector) (VFXGlowOffset :type Vector) (VFXGlow%20Scale :type Vector%202D%20Structure))
  (bind _sm_coin (Variables|Default|GetSM_Coin))
  (bind _returnvalue (Transformation|GetWorldRotation _sm_coin))
  (Utilities|IsValid HidingProp
    (:Is%20Valid
      (Variables|Default|SetAdditionalStreaks AdditionalStreaks)
      (Variables|Default|SetStreaksScaleMultiplier StreaksScaleMultiplier)
      (Variables|Default|SetVFXGlowOffset VFXGlowOffset)
      (Variables|Default|SetVFXStreaks%20Height VFXStreaks%20Height)
      (Variables|Default|SetVFXStreaks%20Radius VFXStreaks%20Radius)
      (Variables|Default|SetVFXStreaks%20Offset VFXStreaks%20Offset)
      (Variables|Default|SetVFXGlow%20Scale VFXGlow%20Scale)
      (Variables|Default|SetHidingProp HidingProp)
      (Development|SetHiddeninGame (Variables|Default|GetSM_Coin) true false)
      (Development|SetHiddeninGame (Variables|Default|GetGlowMesh) true false)
      (bind _output_get (Variables|Default|SetCurrentLocation (+ SpawnOffset (Transformation|GetActorLocation (Variables|Default|GetHidingProp)))))
      (Variables|Default|SetFinalLocation (Transformation|GetWorldLocation _sm_coin))
      (Transformation|SetWorldLocation (Variables|Default|GetSM_Coin) _output_get false false)
      (Variables|Default|SetFinalRotation _returnvalue)
      (Variables|Default|SetCurrentRotation _returnvalue)
      (Variables|Default|SetCurrentScale "0.250000,0.250000,0.250000")
      (Variables|Default|SetFinalScale (Transformation|GetWorldScale _sm_coin)))))

(event EventBeginPlay
  ;[1 1 1] BEGIN PLAY
  (Components|Timeline|SetPlayRate (Variables|QuestVisual_XPCoin_Hidden|GetPickedUp_TL) 2.5))

(event Custom|HidingPropDestroyed
  (bind _returnvalue (Utilities|Time|GetWorldDeltaSeconds))
  (bind _finallocation (Variables|Default|GetFinalLocation))
  (bind _currentlocation (Variables|Default|GetCurrentLocation))
  (Utilities|IsValid (Variables|Default|GetHidingProp)
    (:Is%20Valid
      (Variables|Default|SetHidingPropisDestroyed true)
      (CallFunction|EndSpinning)
      (Development|SetHiddeninGame (Variables|Default|GetSM_Coin) false true)
      (Development|SetHiddeninGame (Variables|Default|GetGlowMesh) false true)
      (Components|Activation|SetActive (Variables|Default|GetTrail) true false)
      ;[1 1 1] Move hat along timeline with vertical and horizontal variance to (poorly) simulate flying
      (bind (_direction _heightpath _rotation _scale) (Timeline|Timeline_0 PlayFromStart)
        (:Update
          (Transformation|SetWorldLocation (Variables|Default|GetSM_Coin) (Math|Vector|vector+vector (Math|Interpolation|VInterpTo _currentlocation _finallocation _returnvalue 0.6) (Math|Vector|MakeVector 0.0 0.0 _heightpath)) false false)
          (Variables|Default|SetCurrentLocation (Math|Vector|Lerp(Vector) _currentlocation _finallocation _returnvalue))
          (Transformation|SetWorldScale3D (Variables|Default|GetSM_Coin) _scale)
          (Variables|Default|SetCurrentScale (Variables|Default|GetCurrentScale))
          (Transformation|SetWorldRotation (Variables|Default|GetSM_Coin) (Math|Rotator|MakeRotator (.roll (Variables|Default|GetFinalRotation)) (.pitch (Variables|Default|GetFinalRotation)) (+ (.yaw (Variables|Default|GetFinalRotation)) _rotation)) false false))
        (:Finished
          (Transformation|SetWorldRotation (Variables|Default|GetSM_Coin) (Variables|Default|GetFinalRotation) false false)
          (CallFunction|BeginSpinning))
        (:FXSpawn
          (bind _returnvalue_2 (Effects|SpawnEmitterAttached :EmitterTemplate (Variables|Default|GetCoinCollectionEmitter) :AttachToComponent (Variables|Default|GetDefaultSceneRoot) :AttachPointName "None" :Location (Transformation|GetSocketLocation (Variables|Default|GetSM_Coin) "FX") :Scale (/ (Variables|Default|GetFinalScale) 0.75) :LocationType "KeepWorldPosition" :bAutoDestroy true :PoolingMethod "None" :bAutoActivate true))
          (Rendering|SetOnlyOwnerSee _returnvalue_2 true :!class /Script/Engine.PrimitiveComponent)
          (Audio|SpawnSoundAttached "/Game/Athena/Sounds/Quests/ScoreCoin_Collect_Gold_Cue.ScoreCoin_Collect_Gold_Cue" (Variables|Default|GetSM_Coin)))))))

(event OnComponentBeginOverlap(SpinOverlap)
  (params (OverlappedComponent :type Primitive%20Component%20Object%20Reference) (OtherActor :type Actor%20Object%20Reference) (OtherComp :type Primitive%20Component%20Object%20Reference) (OtherBodyIndex :type Integer) (bFromSweep :type Boolean) (SweepResult :type Hit%20Result%20Structure%20%28by%20ref%29))
  ;[1 1 1] CLIENT ONLY SPINNING
  (Utilities|Casting|CastToFortPlayerPawnAthena OtherActor)
  (Variables|Default|SetCurrentRelativeScale (Variables|/Script/Engine.SceneComponent|GetRelativeScale3D (Variables|Default|GetSM_Coin)))
  (bind (_direction _risetime _pulsetime) (Timeline|SpinTL Play)
    (:Update
      (Transformation|SetRelativeRotation (Variables|Default|GetSM_Coin) (Math|Rotator|MakeRotator 0.0 0.0 (* _risetime (Variables|Default|GetRelativeRotationAmount))) false false)
      (Transformation|SetRelativeScale3D (Variables|Default|GetSM_Coin) (Math|Vector|Lerp(Vector) (Variables|Default|GetCurrentRelativeScale) (* (Variables|Default|GetCurrentRelativeScale) (Variables|Default|GetPulseSizeMultiplier)) _pulsetime)))
    (:Finished
      (bind _returnvalue (Utilities|Time|SetTimerbyEvent (Utilities|Select :Index _direction :Forward (delegate Custom|LoopForwards) :Backward (delegate Custom|LoopBackwards)) (Variables|Default|GetDelayBetweenSpin)))
      (Variables|Default|SetLoopTimerHandle _returnvalue))
    (:Start%20Spin
      (switch Utilities|FlowControl|Switch|SwitchonETimelineDirection _direction
        (:Forward
          (Audio|SpawnSoundAttached (Variables|Sound|GetCoinSpinSound) (Variables|Default|GetSM_Coin))
          (if (not (Variables|Default|GetHidingPropisDestroyed))
            (bind _returnvalue_2 (Niagara|SpawnSystemAttached (Variables|Default|GetNS_FlareUp) (Variables|Default|GetSM_Coin) "FX" "0, 0, 0" "0, 0, 0" "KeepRelativeOffset" true))
            (Variables|Default|SetNiagara%20Component _returnvalue_2)
            ;[1 1 1] Editable VFX Parameters
            (Niagara|SetNiagaraVariable(Vector2) (Variables|Default|GetNiagara%20Component) "GlowScale" (Variables|Default|GetVFXGlow%20Scale))
            (Effects|Components|ParticleSystem|SetVectorParameter (Variables|Default|GetNiagara%20Component) "VFXGlowOffset" (Variables|Default|GetVFXGlowOffset))
            (Effects|Components|ParticleSystem|SetVectorParameter (Variables|Default|GetNiagara%20Component) "StreaksOffset" (Variables|Default|GetVFXStreaks%20Offset))
            (Effects|Components|ParticleSystem|SetFloatParameter (Variables|Default|GetNiagara%20Component) "StreaksHeight" (Variables|Default|GetVFXStreaks%20Height))
            (Effects|Components|ParticleSystem|SetFloatParameter (Variables|Default|GetNiagara%20Component) "StreaksRadius" (Variables|Default|GetVFXStreaks%20Radius))
            (Effects|Components|ParticleSystem|SetFloatParameter (Variables|Default|GetNiagara%20Component) "AdditionalStreaks" (Variables|Default|GetAdditionalStreaks))
            (Effects|Components|ParticleSystem|SetFloatParameter (Variables|Default|GetNiagara%20Component) "StreaksScaleMultiplier" (Variables|Default|GetStreaksScaleMultiplier))))))))

(event Custom|LoopForwards
  ;[1 1 1] CLIENT ONLY SPINNING
  (Timeline|SpinTL PlayFromStart))

(event Custom|BeginSpinning
  ;[1 1 1] CLIENT ONLY SPINNING
  (Variables|Default|SetCurrentRelativeScale (Variables|/Script/Engine.SceneComponent|GetRelativeScale3D (Variables|Default|GetSM_Coin)))
  (Timeline|SpinTL Play))

(event Custom|LoopBackwards
  ;[1 1 1] CLIENT ONLY SPINNING
  (Timeline|SpinTL ReverseFromEnd))

(event OnComponentEndOverlap(SpinOverlap)
  (params (OverlappedComponent :type Primitive%20Component%20Object%20Reference) (OtherActor :type Actor%20Object%20Reference) (OtherComp :type Primitive%20Component%20Object%20Reference) (OtherBodyIndex :type Integer))
  ;[1 1 1] CLIENT ONLY SPINNING
  (Components|Timeline|Stop (Variables|QuestVisual_XPCoin_Hidden|GetSpinTL))
  (Utilities|Time|ClearandInvalidateTimerbyHandle (Variables|Default|GetLoopTimerHandle)))

(event Custom|EndSpinning
  ;[1 1 1] CLIENT ONLY SPINNING
  (Components|Timeline|Stop (Variables|QuestVisual_XPCoin_Hidden|GetSpinTL))
  (Utilities|Time|ClearandInvalidateTimerbyHandle (Variables|Default|GetLoopTimerHandle)))

(event EventObjectiveCompleted
  (params (PlayerController :type Fort%20Player%20Controller%20Object%20Reference))
  (bind (_relativerotation_roll _relativerotation_pitch _relativerotation_yaw) (Math|Rotator|BreakRotator (Variables|/Script/Engine.SceneComponent|GetRelativeRotation_Roll (Variables|Default|GetSM_Coin))))
  ;[1 1 1] CLIENT ONLY: Collection FX
  (CallFunction|EndSpinning)
  (bind _returnvalue (Rendering|Material|CreateDynamicMaterialInstance (Variables|Default|GetSM_Coin) 0 0 "None" :!class /Script/Engine.PrimitiveComponent))
  (Variables|Default|SetMID0 _returnvalue)
  (bind _returnvalue_2 (Rendering|Material|CreateDynamicMaterialInstance (Variables|Default|GetSM_Coin) 1 0 "None" :!class /Script/Engine.PrimitiveComponent))
  (Variables|Default|SetMID1 _returnvalue_2)
  (Audio|PlaySoundatLocation (Variables|Sound|GetCoinCollectionSound) (Transformation|GetActorLocation))
  (Variables|Default|SetPreShrunkScale (Variables|/Script/Engine.SceneComponent|GetRelativeScale3D (Variables|Default|GetSM_Coin)))
  (Variables|Default|SetCoinRelativeTransform (Variables|/Script/Engine.SceneComponent|GetRelativeLocation (Variables|Default|GetSM_Coin)))
  (Variables|Default|SetCoinRelativeRotationZ _relativerotation_yaw)
  (bind _light (Variables|Default|GetLight)
    (:then
      (Development|SetHiddeninGame _light true false))
    (:else
      _))
  (bind (_direction _scale _zrotation _bobheight _location) (Timeline|PickedUp_TL PlayFromStart)
    (:Update
      (Transformation|SetRelativeLocation (Variables|Default|GetSM_Coin) (+ _location (Variables|Default|GetCoinRelativeTransform)) false false)
      (Transformation|SetRelativeScale3D (Variables|Default|GetSM_Coin) (Math|Vector|Lerp(Vector) (Variables|Default|GetPreShrunkScale) "0, 0, 0" _scale))
      (Transformation|AddRelativeRotation (Variables|Default|GetSM_Coin) (Math|Rotator|MakeRotator 0.0 0.0 (+ _zrotation (Variables|Default|GetCoinRelativeRotationZ))) false false)
      (Rendering|Material|SetVectorParameterValue (Variables|Default|GetMID0) "BobHeightMax" (Math|Color|Lerp(LinearColor) :B (Variables|Default|GetBobHeightMax) :Alpha _bobheight) :!class /Script/Engine.MaterialInstanceDynamic)
      (Rendering|Material|SetVectorParameterValue (Variables|Default|GetMID0) "BobHeightMin" (Math|Color|Lerp(LinearColor) :B (Variables|Default|GetBobHeightMin) :Alpha _bobheight) :!class /Script/Engine.MaterialInstanceDynamic)
      (Rendering|Material|SetScalarParameterValue (Variables|Default|GetMID0) "GlowScale" _bobheight :!class /Script/Engine.MaterialInstanceDynamic))
    (:Finished
      (Actor|SetLifeSpan 5.0)
      (Development|SetHiddeninGame (Variables|Default|GetSM_Coin) true true))
    (:SpawnParticles
      (bind _returnvalue_3 (Effects|SpawnEmitterAttached :EmitterTemplate (Variables|Default|GetCoinCollectionEmitter) :AttachToComponent (Variables|Default|GetDefaultSceneRoot) :AttachPointName "None" :Location (Transformation|GetSocketLocation (Variables|Default|GetSM_Coin) "FX") :Scale (/ (Variables|Default|GetPreShrunkScale) 0.75) :LocationType "KeepWorldPosition" :bAutoDestroy true :PoolingMethod "None" :bAutoActivate true))
      (Rendering|SetOnlyOwnerSee _returnvalue_3 true :!class /Script/Engine.PrimitiveComponent))))

(fn setNiagaraSystemUserVariable(Color)
  (params (Name :type Name) (Value :type Linear%20Color%20Structure))
  (ManagedParticles|Parameters|%28defunct%29SetColorParameter%28ManagedParticles%29 0 (ManagedParticles|GetNiagaraParticles) Name Value))

(fn ConstructionScript
  (Networking|CanExecuteCosmeticEvents
    (:True
      (if (or (Variables|Default|GetUseLight) (not (Fortnite|IsMobilePlatform)))
        (Utilities|EffectsQualityCheck 0
          (:Cinematic
            _)
          (:Epic
            _)
          (:High
            _)
          (:Medium
            (Components|DestroyComponent (Variables|Default|GetLight)))
          (:Low
            _))
        (Rendering|Components|Light|SetLightColor (Variables|Default|GetLight) (Variables|Default|GetLightColor) true)
        (else
          _)))
    (:False
      _)))
