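# Niagara system for a blue upward flare effect, named for coins and filed with the quest item-collection interactables.
# NE_Glow bursts a single sprite above the origin that swells and fades through curves.
# NE_Streaks emits short-lived, randomly rotated streak sprites from a cylinder volume, launched upward with drag and an alpha fade.
# User parameters expose glow scale and offset, plus streak count, size, radius, height and offset, for per-use tuning.
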NiagaraSystem /Game/Athena/Items/QuestInteractables/ItemCollection/VFX/NS_CoinFlareUp
dependencies {usedDataInterfaces [&NiagaraDataInterfaceCurve], usedDynamicInputs [], usedModules [&SystemState]}
emitters {
 NE_Glow {
  emitterUpdateScript [
   &EmitterState(Life Cycle Mode=::NewEnumerator1, Loop Behavior=::NewEnumerator1, Loop Duration=1, UseLoopDelay=false)
   &SpawnBurst_Instantaneous(Spawn Count=1, Use Spawn Probability=false)
  ]
  particleSpawnScript [/Niagara/Modules/Spawn/Initialization/InitializeParticle(
    Lifetime=0.8
    Module.Write Color=true
    Module.Write Lifetime=true
    Module.Write Mass=true
    Module.Write Material Random=true
    Module.Write Position=true
    Module.Write Scale=false
    Module.Write SpriteRotation=false
    Module.Write SpriteSize=true
    Position=&Add_Vector(A=&SimulationPosition(), B=&Add_Vector(A=<0 0 30>, B=$User.VFXGlowOffset))
    Sprite Size=<50 50>
   )]
  particleUpdateScript [
   &ParticleState()
   &SolveForcesAndVelocity(Clamp Velocity=false, Limit Acceleration=false)
   &Color(
    Color=<rgba 0 0.731 3 1>
    Scale Alpha=&FloatFromCurve(
     FloatCurve=DataInterface() {propertyValues {Curve {keys [{interpMode=RCIM_Cubic, tangentMode=RCTM_Break, time=0, value=0}, <at 0.2011 1>, <at 1 0>]}}}
     Scale Curve=0.7
    )
   )
   &ScaleSpriteSize(Scale Factor=&Multiply_Vector2(
     A=$User.GlowScale
     B=&Vector2DFromFloat(Value=&FloatFromCurve(FloatCurve=DataInterface() {propertyValues {Curve {keys [{interpMode=RCIM_Cubic, tangentMode=RCTM_Break, time=0, value=0}, <at 0.1794 1>, <at 1 0.5>]}}}))
    ))
  ]
  rendererClasses [&NiagaraSpriteRendererProperties]
  renderers [{rendererClass=&NiagaraSpriteRendererProperties}]
 }
 NE_Streaks {
  emitterUpdateScript [
   &EmitterState(Life Cycle Mode=::NewEnumerator1, Loop Behavior=::NewEnumerator1, Loop Duration=0.5, UseLoopDelay=false)
   &SpawnRate(SpawnRate=&Add_Float(A=25, B=$User.AdditionalStreaks), Use Spawn Probability=false)
  ]
  particleSpawnScript [
   &InitializeParticle(
    Lifetime=&UniformRangedFloat(Minimum=0.1, Maximum=0.4)
    Mass=&UniformRangedFloat(Minimum=0.6, Maximum=1.2)
    Module.Write Color=true
    Module.Write Lifetime=true
    Module.Write Mass=true
    Module.Write Material Random=true
    Module.Write Position=false
    Module.Write Scale=false
    Module.Write SpriteRotation=true
    Module.Write SpriteSize=true
    Position=&SimulationPosition()
    Sprite Rotation=&NormalizedAngleToDegrees(NormalizedAngle=&UniformRangedFloat(Minimum=0, Maximum=1))
    Sprite Size=&Multiply_Vector2(A=<5 50>, B=&Vector2DFromFloat(Value=$User.StreaksScaleMultiplier))
   )
   /Niagara/Modules/Spawn/Location/SphereLocation(MaskSpawnGroup=false, Module.Override Randomness=false, Sphere Radius=8, Surface Only=false) {enabled=false}
   &CylinderLocation(
    Cylinder Height:NiagaraFloat=$User.StreaksHeight
    Cylinder Radius=$User.StreaksRadius
    MaskSpawnGroup=false
    Module.Enable Lathe Profile=false
    Module.Override Randomness=false
    Module.Surface Only=false
    Offset=$User.StreaksOffset
   )
   &AddVelocity(Velocity=<0 0 450>)
  ]
  particleUpdateScript [
   &ParticleState()
   &Color(
    Color=<rgba 0 0.4873 2 1>
    Scale Alpha=&FloatFromCurve(FloatCurve=DataInterface() {propertyValues {Curve {keys [
         {interpMode=RCIM_Cubic, tangentMode=RCTM_Break, time=0, value=0}
         {interpMode=RCIM_Cubic, tangentMode=RCTM_Break, time=0.25, value=1}
         <at 1 0>
        ]}}})
   )
   &Drag(Drag=0.25, Use Linear Drag=true, Use Rotational Drag=false)
   &SolveForcesAndVelocity(Clamp Velocity=false, Limit Acceleration=false)
  ]
  rendererClasses [&NiagaraSpriteRendererProperties]
  renderers [{rendererClass=&NiagaraSpriteRendererProperties}]
 }
}
summary {
 systemName=NS_CoinFlareUp
 emitters [
  {emitterName=NE_Streaks, rendererClasses [&NiagaraSpriteRendererProperties]}
  {emitterName=NE_Glow, rendererClasses [&NiagaraSpriteRendererProperties]}
 ]
 userVariables [
  User.AdditionalStreaks {type=NiagaraFloat, defaultValue=50, description=""}
  User.GlowScale {type=Vector2f, description="", defaultValue {struct=Vector2f, value=<5 8>}}
  User.StreaksHeight {type=NiagaraFloat, defaultValue=70, description=""}
  User.StreaksOffset {type=Vector3f, description="", defaultValue {struct=Vector3f, value=<0 0 0>}}
  User.StreaksRadius {type=NiagaraFloat, defaultValue=45, description=""}
  User.StreaksScaleMultiplier {type=NiagaraFloat, defaultValue=1, description=""}
  User.VFXGlowOffset {type=Vector3f, description="", defaultValue {struct=Vector3f, value=<0 0 0>}}
 ]
}
